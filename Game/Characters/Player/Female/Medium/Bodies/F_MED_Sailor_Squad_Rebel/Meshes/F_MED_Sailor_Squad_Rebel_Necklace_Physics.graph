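# Physics asset for the necklace on the female medium Sailor Squad Rebel player body, set up for dangling motion.
# A chain of small capsule bodies runs from dyn_necklace_a_base through links a to d, each link joined by a twist-locked constraint with narrow, asymmetric swing limits.
# The base attaches to spine_05 with free swing and twist, and two large overlapping capsules on spine_05 approximate the upper torso.
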PhysicsAsset /Game/Characters/Player/Female/Medium/Bodies/F_MED_Sailor_Squad_Rebel/Meshes/F_MED_Sailor_Squad_Rebel_Necklace_Physics
bodies {
 dyn_necklace_a [{shapeType=Capsule, length=3.25, radius=1, center=<2.2892 0 0>, rotation=<rot 90 0 0>}]
 dyn_necklace_a_base [{shapeType=Capsule, length=0.505, radius=0.505, rotation=<rot 0 -90 0>}]
 dyn_necklace_b [{shapeType=Capsule, length=2.25, radius=1, center=<2 0 0>, rotation=<rot 90 0 0>}]
 dyn_necklace_c [{shapeType=Capsule, length=2.25, radius=1, center=<2 0 0>, rotation=<rot 90 0 0>}]
 dyn_necklace_d [{shapeType=Capsule, length=2.25, radius=1, center=<2 0 0>, rotation=<rot 90 0 0>}]
 spine_05 [
  {shapeType=Capsule, length=18.5, radius=8.75, center=<-4.4192 1.5503 0>}
  {shapeType=Capsule, length=15.5, radius=8.5, center=<1.4717 -2.2564 0>}
 ]
}
constraints [
 {
  bone1Name=dyn_necklace_a
  bone2Name=dyn_necklace_a_base
  swing1LimitDegrees=20
  swing2LimitDegrees=10
  twistLimitDegrees=45
  twistMotion=Locked
 }
 {bone1Name=dyn_necklace_b, bone2Name=dyn_necklace_a, swing1LimitDegrees=20, swing2LimitDegrees=10, twistLimitDegrees=45, twistMotion=Locked}
 {bone1Name=dyn_necklace_c, bone2Name=dyn_necklace_b, swing1LimitDegrees=20, swing2LimitDegrees=10, twistLimitDegrees=45, twistMotion=Locked}
 {bone1Name=dyn_necklace_d, bone2Name=dyn_necklace_c, swing1LimitDegrees=20, swing2LimitDegrees=10, twistLimitDegrees=45, twistMotion=Locked}
 {
  bone1Name=dyn_necklace_a_base
  bone2Name=spine_05
  swing1LimitDegrees=45
  swing1Motion=Free
  swing2LimitDegrees=45
  swing2Motion=Free
  twistLimitDegrees=45
  twistMotion=Free
 }
]
