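# Physics asset for the female medium Sailor Squad Rebel player body, covering the pelvis, thighs and dangling gear: a grenade, two pouches and a two-segment hiplace piece.
# Tiny capsule base bodies and both thighs link to the pelvis with free angular motion.
# Gear bodies hang from those bases with tight swing and twist limits.
# The second hiplace box chains off the first, whose twist is locked.
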PhysicsAsset /Game/Characters/Player/Female/Medium/Bodies/F_MED_Sailor_Squad_Rebel/Meshes/F_MED_Sailor_Squad_Rebel_Physics
bodies {
 dyn_grenade_base_r [{shapeType=Capsule, length=0.505, radius=0.505, rotation=<rot 0 -90 0>}]
 dyn_grenade_r [{shapeType=Capsule, length=10.25, radius=4, center=<4.2121 -0.0001 -3.8163>, rotation=<rot 90 0 0>}]
 dyn_hiplace_a_base_r [{shapeType=Capsule, length=0.505, radius=0.505, rotation=<rot 0 -90 0>}]
 dyn_hiplace_a_r [{shapeType=Box, extentX=7.5, extentY=16.5, extentZ=3.5, center=<4 0 -2>}]
 dyn_hiplace_b_r [{shapeType=Box, extentX=7.5, extentY=16.5, extentZ=3.5, center=<4 0 -2>}]
 dyn_pouch [{shapeType=Box, extentX=5.5012, extentY=16.4172, extentZ=25.0954, center=<8.8367 1.619 -0.3966>, rotation=<rot -4.3402 -89.9995 -0.8156>}]
 dyn_pouch_base [{shapeType=Capsule, length=0.505, radius=0.505, rotation=<rot 0 -90 0>}]
 dyn_pouch_base_r [{shapeType=Capsule, length=0.505, radius=0.505, rotation=<rot 0 -90 0>}]
 dyn_pouch_r [{shapeType=Capsule, length=9.9009, radius=2.8152, center=<3.3572 -0.539 -2.6693>, rotation=<rot 0 -92.744 -86.7747>}]
 pelvis [{shapeType=Capsule, length=9.5, radius=9.5, center=<-0.9772 -1.9219 0>}]
 thigh_l [{shapeType=Capsule, length=22.5, radius=8.5, center=<-18.7286 0 0>, rotation=<rot 90 2.1476 6.8589>}]
 thigh_r [{shapeType=Capsule, length=23.75, radius=9, center=<18.7286 0 -0.379>, rotation=<rot 87.1857 175.2879 0>}]
}
constraints [
 {
  bone1Name=thigh_l
  bone2Name=pelvis
  swing1LimitDegrees=45
  swing1Motion=Free
  swing2LimitDegrees=45
  swing2Motion=Free
  twistLimitDegrees=45
  twistMotion=Free
 }
 {
  bone1Name=thigh_r
  bone2Name=pelvis
  swing1LimitDegrees=45
  swing1Motion=Free
  swing2LimitDegrees=45
  swing2Motion=Free
  twistLimitDegrees=45
  twistMotion=Free
 }
 {
  bone1Name=dyn_grenade_base_r
  bone2Name=pelvis
  swing1LimitDegrees=45
  swing1Motion=Free
  swing2LimitDegrees=45
  swing2Motion=Free
  twistLimitDegrees=45
  twistMotion=Free
 }
 {bone1Name=dyn_grenade_r, bone2Name=dyn_grenade_base_r, swing1LimitDegrees=1, swing2LimitDegrees=30, twistLimitDegrees=2}
 {
  bone1Name=dyn_pouch_base_r
  bone2Name=pelvis
  swing1LimitDegrees=45
  swing1Motion=Free
  swing2LimitDegrees=45
  swing2Motion=Free
  twistLimitDegrees=45
  twistMotion=Free
 }
 {bone1Name=dyn_pouch_r, bone2Name=dyn_pouch_base_r, swing1LimitDegrees=2, swing2LimitDegrees=25, twistLimitDegrees=2}
 {
  bone1Name=dyn_pouch_base
  bone2Name=pelvis
  swing1LimitDegrees=45
  swing1Motion=Free
  swing2LimitDegrees=45
  swing2Motion=Free
  twistLimitDegrees=45
  twistMotion=Free
 }
 {bone1Name=dyn_pouch, bone2Name=dyn_pouch_base, swing1LimitDegrees=30, swing2LimitDegrees=30, twistLimitDegrees=15}
 {
  bone1Name=dyn_hiplace_a_base_r
  bone2Name=pelvis
  swing1LimitDegrees=45
  swing1Motion=Free
  swing2LimitDegrees=45
  swing2Motion=Free
  twistLimitDegrees=45
  twistMotion=Free
 }
 {
  bone1Name=dyn_hiplace_a_r
  bone2Name=dyn_hiplace_a_base_r
  swing1LimitDegrees=5
  swing2LimitDegrees=20
  twistLimitDegrees=10
  twistMotion=Locked
 }
 {bone1Name=dyn_hiplace_b_r, bone2Name=dyn_hiplace_a_r, swing1LimitDegrees=10, swing2LimitDegrees=20, twistLimitDegrees=10}
]
